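# Debug widget for proximity chat that shows diagnostic text on screen, built on the native CreativeProximityChatDebugWidget class.
# Every Tick it sets a CommonTextBlock inside a SizeBox to the text returned by GetProximityChatDebugText.
# The Blueprint only handles display, with no variables of its own and unwired PreConstruct and Construct events.
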WidgetBlueprint /CreativeCommunication/ProximityChat/ProximityChatDebugWidget
graphs [EventGraph {nodes [
   Event Tick()
   > "SetText (Text)"(
    self=Get ProximityChatDebugInfo() -> (ProximityChatDebugInfo:Common Text Object Reference)
    InText=GetProximityChatDebugText!() -> (ReturnValue:Text)
   )
   Event PreConstruct()
   Event Construct()
  ]}]
variables []
widgets {
 info {
  inheritedWidgetCount=1
  parentClass=/Script/CreativeCommunicationUI.CreativeProximityChatDebugWidget
  rootWidgetClass=&SizeBox
  widgetCount=3
 }
 widgets [
  SizeBox_0 {widgetClassPath=&SizeBox, widget=SizeBox_0}
  ProximityChatDebugInfo {
   widgetClassPath=&CommonTextBlock
   bIsVariable=true
   parent=SizeBox_0
   widget=ProximityChatDebugInfo
   properties {
    bIsScrollingEnabled=false
    text=Text Block
    font {
     bForceMonospaced=false
     bMaterialIsStencil=false
     fontMaterial=null
     fontObject=&HeadingNow
     letterSpacing=0
     monospacedWidth=1
     size=13.5
     skewAmount=0
     typefaceFontName=75 Medium
     outlineSettings {
      bApplyOutlineToDropShadows=false
      bMiteredCorners=false
      bSeparateFillAlpha=false
      outlineMaterial=null
      outlineSize=0
      outlineColor=<rgba 0 0 0 1>
     }
    }
    shadowOffset=<0 0>
   }
   slot {class=SizeBoxSlot}
  }
  TutorialHighlight {widgetClassPath=&TutorialHighlight}
 ]
}
